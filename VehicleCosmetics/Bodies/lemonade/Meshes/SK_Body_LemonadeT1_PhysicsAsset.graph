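# Physics asset for the Lemonade T1 vehicle body skeletal mesh, a car cosmetic with bodies for chassis, glass, suspension and wheels.
# Every body is a single capsule, with only a tiny one on the root-like frame bone.
# Constraints form a tree from frame to body and glass, front differential to knuckles and wheels, and rear axle pivots to wheels, sharing identical swing and twist limit values.
# Only the front-left side has a shock base body.
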PhysicsAsset /VehicleCosmetics/Bodies/lemonade/Meshes/SK_Body_LemonadeT1_PhysicsAsset
bodies {
 axle_pivot_bk_l [{shapeType=Capsule, length=27.5909, radius=26.6723, center=<-2.1395 -39.0455 12.945>, rotation=<rot 0 159.7299 -165.3976>}]
 axle_pivot_bk_r [{shapeType=Capsule, length=27.5909, radius=26.6723, center=<-2.1395 39.0455 12.945>, rotation=<rot 0 20.2701 -165.3976>}]
 body [{shapeType=Capsule, length=233.5466, radius=105.1808, center=<15.9204 -0.146 15.1633>, rotation=<rot 0 -90.053 -85.351>}]
 frame [{shapeType=Capsule, length=0.505, radius=0.505, rotation=<rot 0 -90 0>}]
 front_differential [{shapeType=Capsule, length=57.6682, radius=9.1794, center=<-0.0005 0 0.0005>, rotation=<rot 0 0 -90>}]
 glass_c [{shapeType=Capsule, length=71.0339, radius=20.6527, center=<-17.9049 0 13.5369>, rotation=<rot 0 -180 -90>}]
 glass_l [{shapeType=Capsule, length=51.9667, radius=14.1613, center=<10.4197 4.7674 10.2756>, rotation=<rot 0 -88.7598 -83.9184>}]
 glass_r [{shapeType=Capsule, length=51.9667, radius=14.1613, center=<-10.4197 -4.7674 -10.2757>, rotation=<rot 0 -88.7598 -83.9184>}]
 shock_base_fr_l [{shapeType=Capsule, length=14.2513, radius=8.5532, center=<-0.1522 1.1873 28.5811>, rotation=<rot 4.9731 -90 -1.3502>}]
 steering_knuckle_fr_l [{shapeType=Capsule, length=26.2103, radius=17.6196, center=<-1.1962 -21.0261 5.9626>, rotation=<rot 0 -22.4181 -80.0933>}]
 steering_knuckle_fr_r [{shapeType=Capsule, length=28.0604, radius=25.466, center=<-0.5312 18.8211 11.7869>, rotation=<rot 0 21.4844 -62.877>}]
 wheel_disc_bk_l [{shapeType=Capsule, length=20.77, radius=20.77, center=<-0.0624 -1.7845 0.0584>, rotation=<rot 0 -90 -89.9999>}]
 wheel_disc_bk_r [{shapeType=Capsule, length=20.77, radius=20.77, center=<-0.0624 1.7845 0.0584>, rotation=<rot 0 -90 -89.9999>}]
 wheel_disc_fr_l [{shapeType=Capsule, length=20.77, radius=20.77, center=<-0.0627 -1.7815 0.0582>, rotation=<rot 0 -90 -90>}]
 wheel_disc_fr_r [{shapeType=Capsule, length=20.77, radius=20.77, center=<-0.0627 1.7815 0.0582>, rotation=<rot 0 -90 -89.9998>}]
}
constraints [
 {bone1Name=body, bone2Name=frame, swing1LimitDegrees=45, swing2LimitDegrees=45, twistLimitDegrees=45}
 {bone1Name=glass_c, bone2Name=body, swing1LimitDegrees=45, swing2LimitDegrees=45, twistLimitDegrees=45}
 {bone1Name=glass_l, bone2Name=body, swing1LimitDegrees=45, swing2LimitDegrees=45, twistLimitDegrees=45}
 {bone1Name=glass_r, bone2Name=body, swing1LimitDegrees=45, swing2LimitDegrees=45, twistLimitDegrees=45}
 {bone1Name=front_differential, bone2Name=frame, swing1LimitDegrees=45, swing2LimitDegrees=45, twistLimitDegrees=45}
 {bone1Name=steering_knuckle_fr_l, bone2Name=front_differential, swing1LimitDegrees=45, swing2LimitDegrees=45, twistLimitDegrees=45}
 {bone1Name=wheel_disc_fr_l, bone2Name=steering_knuckle_fr_l, swing1LimitDegrees=45, swing2LimitDegrees=45, twistLimitDegrees=45}
 {bone1Name=shock_base_fr_l, bone2Name=steering_knuckle_fr_l, swing1LimitDegrees=45, swing2LimitDegrees=45, twistLimitDegrees=45}
 {bone1Name=steering_knuckle_fr_r, bone2Name=front_differential, swing1LimitDegrees=45, swing2LimitDegrees=45, twistLimitDegrees=45}
 {bone1Name=wheel_disc_fr_r, bone2Name=steering_knuckle_fr_r, swing1LimitDegrees=45, swing2LimitDegrees=45, twistLimitDegrees=45}
 {bone1Name=axle_pivot_bk_l, bone2Name=frame, swing1LimitDegrees=45, swing2LimitDegrees=45, twistLimitDegrees=45}
 {bone1Name=wheel_disc_bk_l, bone2Name=axle_pivot_bk_l, swing1LimitDegrees=45, swing2LimitDegrees=45, twistLimitDegrees=45}
 {bone1Name=axle_pivot_bk_r, bone2Name=frame, swing1LimitDegrees=45, swing2LimitDegrees=45, twistLimitDegrees=45}
 {bone1Name=wheel_disc_bk_r, bone2Name=axle_pivot_bk_r, swing1LimitDegrees=45, swing2LimitDegrees=45, twistLimitDegrees=45}
]
